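# Mono looping MetaSound source for the hot air balloon's burner, layering a pitched-down loop with random burner one-shots.
# One-shots retrigger after random delays through a deferred trigger variable, and their duration-scaled AD envelope ducks the loop, which also wavers with an LFO.
# An arrive/departure sidechain control bus scales the final output.
# The Mobile page keeps only the loop, LFO and sidechain; the PitchSineDuration input is unconnected.
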MetaSoundSource /HotAirBalloon/Assets/Audio/MSS_HotAirBalloon_Burner
inputs [
 PitchSineDuration {type=Float}
 UE.Source.OnPlay {type=Trigger}
 VolSineDuration {type=Float}
]
interfaces [UE.OutputFormat.Mono:1.0, UE.Source:1.1]
outputs ["UE.OutputFormat.Mono.Audio:0" {type=Audio}]
pages [
 Default {
  is_build_graph=true
  input_defaults {PitchSineDuration=30, UE.Source.OnPlay=null, VolSineDuration=20}
  nodes [
   n1 = VariableDeferredAccessor.Trigger(Variable=InitVariable.Trigger().Variable)
   TriggerAnyTriggerAny2_0 = "TriggerAny.Trigger Any (2)"(In 0=^UE.Source.OnPlay, In 1=@n1.Value)
   ArrayRandomGet_0 = Array.Random Get:WaveAsset:Array(Next=@TriggerAnyTriggerAny2_0.Out, In Array=VariableAccessor.WaveAsset:Array(Variable=InitVariable.WaveAsset:Array().Variable).Value) {also {defaults {Weights []}}}
   GetWaveDuration_0 = "UE.Get Wave Duration: "(Wave=@ArrayRandomGet_0.Value)
   WavePlayer_1 = UE.Wave Player:Mono(Wave Asset=@ArrayRandomGet_0.Value, Play=UE.RandomFloat(Next=@ArrayRandomGet_0.On Next, Reset=@ArrayRandomGet_0.On Reset).On Next) {also {version [1,1]}}
   ADEnvelopeADEnvelope_0 = AD Envelope.AD Envelope:Float(
    Attack Time=UE.Multiply:Time by Float(PrimaryOperand=@GetWaveDuration_0.Duration) {also {defaults {AdditionalOperands=0.1}}}.Out
    Decay Time=UE.Multiply:Time by Float(PrimaryOperand=@GetWaveDuration_0.Duration) {also {defaults {AdditionalOperands=0.9}}}.Out
    Trigger=@WavePlayer_1.On Play
   )
   Multiply_1 = UE.Multiply:Float(
    PrimaryOperand=UE.InterpTo:Audio(Target=MapRange.MapRange:Float(In=@ADEnvelopeADEnvelope_0.Out Envelope) {also {defaults {Out Range A=1, Out Range B=0.25}}}.Out Value).Value
    AdditionalOperands=UE.LFO:Audio(Frequency=UE.Divide:Float(AdditionalOperands=^VolSineDuration).Out) {also {defaults {Min Value=0.75}}}.Out
   )
   WavePlayer_0 = UE.Wave Player:Mono(Play=^UE.Source.OnPlay) {also {
     defaults {Loop="<MetaSoundBuilderResult.SUCCEEDED: 0>", Pitch Shift=-1.5, Wave Asset=/HotAirBalloon/Assets/Audio/Waves/Veh_HotairBalloon_Burner_Loop}
     version [1,1]
    }}
   AudioMixerAudioMixerMono2_0 = "AudioMixer.Audio Mixer (Mono, 2)"(Gain 1=@Multiply_1.Out, In 0=UE.Reroute(Value=@WavePlayer_1.Out Mono).Value, In 1=@WavePlayer_0.Out Mono)
   Multiply_0 = UE.Multiply:Audio by Float(
    PrimaryOperand=@AudioMixerAudioMixerMono2_0.Out
    AdditionalOperands=Modulation.GetModulatorValue() {also {defaults {Modulator=&/CB_HotAirBalloon_ArriveDeparture_Sidechain}}}.Out
   )
   n2 = UE.RandomTime(Next=@WavePlayer_1.On Finished) {also {defaults {Max=20, Min=5}}}
   TriggerDelay_0 = UE.Trigger Delay(In=@n2.On Next, Delay Time=@n2.Value)
   VariableMutator.Trigger(Value=@TriggerDelay_0.Out, Variable=@n1.Variable)
  ]
  links [{dst="@UE.OutputFormat.Mono.Audio:0", src=Multiply_0.Out}]
 }
 Mobile {
  nodes [
   InitVariable.Trigger()
   InitVariable.WaveAsset:Array()
   WavePlayer_0 = UE.Wave Player:Mono(Play=^UE.Source.OnPlay) {also {
     defaults {Loop="<MetaSoundBuilderResult.SUCCEEDED: 0>", Pitch Shift=-1.5, Wave Asset=&Veh_HotairBalloon_Burner_Loop}
     version [1,1]
    }}
   Multiply_0 = UE.Multiply:Audio by Float(
    AdditionalOperands=UE.LFO:Audio(Frequency=UE.Divide:Float(AdditionalOperands=^VolSineDuration).Out) {also {defaults {Min Value=0.75}}}.Out
    PrimaryOperand=@WavePlayer_0.Out Mono
   )
   Multiply_1 = UE.Multiply:Audio by Float(
    AdditionalOperands=Modulation.GetModulatorValue() {also {defaults {Modulator=&CB_HotAirBalloon_ArriveDeparture_Sidechain}}}.Out
    PrimaryOperand=@Multiply_0.Out
   )
  ]
  links [{dst="@UE.OutputFormat.Mono.Audio:0", src=Multiply_1.Out}]
 }
]
properties {
 assetImportData=&self:AssetImportData
 attenuationSettings=&/ATT_HotAirBalloon_Burner
 bDebug=true
 bLooping=true
 bOverrideConcurrency=true
 duration=10000
 maxDistance=9000
 sampleRate=48000
 soundClassObject=&Fort_Vehicles_Exterior
 concurrencyOverrides {
  bEnableMaxCountPlatformScaling=true
  maxCount=3
  retriggerTime=1
  voiceStealReleaseTime=0.5
  volumeScale=0.5
  volumeScaleAttackTime=1
  platformMaxCount {default=3, perPlatform {mobile=1, switch=1}}
 }
 concurrencySet [null]
}
