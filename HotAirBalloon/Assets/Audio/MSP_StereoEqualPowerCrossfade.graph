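# MetaSound patch that crossfades two stereo inputs into one stereo output by attenuation distance, used in the HotAirBalloon audio.
# Distance between Range.Min and the active sound's MaxDistance is remapped to evaluate two equal-power crossfade wavetables, which set the stereo mixer's gains.
# On OnPlay, UseLogCurve is compared to route one of two table pairs from WTB_Fort_EqualPowerCrossfade.
# The occlusion input is unused.
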MetaSoundPatch /HotAirBalloon/Assets/Audio/MSP_StereoEqualPowerCrossfade
inputs [
 ActiveSound.bIsOccluded {type=Bool}
 ActiveSound.MaxDistance {type=Float}
 In 0 L {type=Audio}
 In 0 R {type=Audio}
 In 1 L {type=Audio}
 In 1 R {type=Audio}
 OnPlay {type=Trigger}
 Range.Min {type=Float}
 UE.Attenuation.Distance {type=Float}
 UseLogCurve {type=Bool}
]
interfaces [Fort.ActiveSound:1.0, UE.Attenuation:1.0]
outputs [
 Out L {type=Audio}
 Out R {type=Audio}
]
pages [Default {
  input_defaults {
   ActiveSound.MaxDistance=0
   ActiveSound.bIsOccluded=null
   In 0 L=null
   In 0 R=null
   In 1 L=null
   In 1 R=null
   OnPlay=null
   Range.Min=100
   UE.Attenuation.Distance=0
   UseLogCurve="<MetaSoundBuilderResult.SUCCEEDED: 0>"
  }
  nodes [
   n1 = MapRange.MapRange:Float(In Range B=^ActiveSound.MaxDistance, In Range A=^Range.Min, In=^UE.Attenuation.Distance)
   n2 = TriggerCompare.Clamp:Bool(Compare=^OnPlay, A=^UseLogCurve) {also {defaults {B="<MetaSoundBuilderResult.SUCCEEDED: 0>"}}}
   TriggerRouteTriggerRouteFloat2_0 = "TriggerRoute.Trigger Route (Float, 2):Float"(Set 1=@n2.False, Set 0=@n2.True) {also {defaults {Value 0=1, Value 1=3}}}
   n3 = TriggerCompare.Clamp:Bool(Compare=^OnPlay, A=^UseLogCurve) {also {defaults {B="<MetaSoundBuilderResult.SUCCEEDED: 0>"}}}
   TriggerRouteTriggerRouteFloat2_1 = "TriggerRoute.Trigger Route (Float, 2):Float"(Set 1=@n3.False, Set 0=@n3.True) {also {defaults {Value 0=0, Value 1=2}}}
   AudioMixerAudioMixerStereo2_0 = "AudioMixer.Audio Mixer (Stereo, 2)"(
    In 0 L=^In 0 L
    In 0 R=^In 0 R
    In 1 L=^In 1 L
    In 1 R=^In 1 R
    Gain 0=UE.WaveTableEvaluate(
     Input=@n1.Out Value
     WaveTable=UE.MetasoundWaveTableGet(TableIndex=@TriggerRouteTriggerRouteFloat2_0.Value) {also {defaults {WaveTableBank=&/WTB_Fort_EqualPowerCrossfade}, version [1,1]}}.Out
    ).Output
    Gain 1=UE.WaveTableEvaluate(
     Input=@n1.Out Value
     WaveTable=UE.MetasoundWaveTableGet(TableIndex=@TriggerRouteTriggerRouteFloat2_1.Value) {also {defaults {WaveTableBank=&WTB_Fort_EqualPowerCrossfade}, version [1,1]}}.Out
    ).Output
   )
  ]
  links [{dst="@Out L", src=AudioMixerAudioMixerStereo2_0.Out L}, {dst="@Out R", src=AudioMixerAudioMixerStereo2_0.Out R}]
 }]
