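# One-shot stereo MetaSound for the hot air balloon explosion, routed to the explosions sound class with attenuation and concurrency limits.
# On play it starts a close wave at reduced gain and, after a short delay, a distant wave.
# An equal-power crossfade patch blends the two layers, fed by listener distance, max distance and occlusion state.
# The sound reports finished once both wave players have finished.
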MetaSoundSource /HotAirBalloon/Assets/Audio/MSS_HotAirBalloon_Explosion
inputs [
 ActiveSound.bIsOccluded {type=Bool}
 ActiveSound.MaxDistance {type=Float}
 UE.Attenuation.Distance {type=Float}
 UE.Source.OnPlay {type=Trigger}
]
interfaces [UE.OutputFormat.Stereo:1.0, UE.Source.OneShot:1.0, UE.Source:1.1]
outputs [
 "UE.OutputFormat.Stereo.Audio:0" {type=Audio}
 "UE.OutputFormat.Stereo.Audio:1" {type=Audio}
 UE.Source.OneShot.OnFinished {type=Trigger}
]
pages [Default {
  input_defaults {ActiveSound.MaxDistance=0, ActiveSound.bIsOccluded=null, UE.Attenuation.Distance=800, UE.Source.OnPlay=null}
  nodes [
   TriggerDelay_0 = UE.Trigger Delay(In=^UE.Source.OnPlay) {also {defaults {Delay Time=0.1}}}
   WavePlayer_1 = UE.Wave Player:Stereo(Play=@TriggerDelay_0.Out) {also {defaults {Wave Asset=/HotAirBalloon/Assets/Audio/Waves/Veh_HotAirBalloon_Explosion_Distant}, version [1,1]}}
   n1 = &MSP_StereoGain(Left In=@WavePlayer_1.Out Left, Right In=@WavePlayer_1.Out Right)
   WavePlayer_0 = UE.Wave Player:Stereo(Play=^UE.Source.OnPlay) {also {defaults {Wave Asset=/HotAirBalloon/Assets/Audio/Waves/Veh_HotAirBalloon_Explosion_Close}, version [1,1]}}
   n2 = &MSP_StereoGain(Left In=@WavePlayer_0.Out Left, Right In=@WavePlayer_0.Out Right) {also {defaults {"Gain (Linear)"=0.5}}}
   MSP_StereoEqualPowerCrossfade_0 = &/MSP_StereoEqualPowerCrossfade(
    ActiveSound.MaxDistance=^ActiveSound.MaxDistance
    ActiveSound.bIsOccluded=^ActiveSound.bIsOccluded
    UE.Attenuation.Distance=^UE.Attenuation.Distance
    OnPlay=^UE.Source.OnPlay
    In 1 L=@n1.Left Out
    In 1 R=@n1.Right Out
    In 0 L=@n2.Left Out
    In 0 R=@n2.Right Out
   )
   TriggerAccumulatorTriggerAccumulate2_0 = "TriggerAccumulator.Trigger Accumulate (2)"(In 0=@WavePlayer_0.On Finished, In 1=@WavePlayer_1.On Finished)
  ]
  links [
   {dst="@UE.OutputFormat.Stereo.Audio:0", src=MSP_StereoEqualPowerCrossfade_0.Out L}
   {dst="@UE.OutputFormat.Stereo.Audio:1", src=MSP_StereoEqualPowerCrossfade_0.Out R}
   {dst="@UE.Source.OneShot.OnFinished", src=TriggerAccumulatorTriggerAccumulate2_0.Out}
  ]
 }]
properties {
 assetImportData=&self:AssetImportData
 attenuationSettings=&/ATT_HotAirBalloon_Explosion
 maxDistance=9500
 numChannels=2
 outputFormat=Stereo
 sampleRate=48000
 soundClassObject=/Game/Sounds/FortSoundClassesAndModes/Fort_Events_Explosions
 concurrencySet [/Game/Sounds/Concurrency_Rules/ThreeMaximum_Owner_PreventNew]
}
