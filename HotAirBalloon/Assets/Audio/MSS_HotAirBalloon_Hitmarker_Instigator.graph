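# One-shot stereo MetaSound for the instigator hitmarker on the hot air balloon vehicle.
# On play it randomly picks one of five impact waves, plays it and passes it through an azimuth panner driven by the spatialization inputs.
# Routing uses the hitmarker sound class, attenuation and concurrency, with hit-notify volume modulation and sends to weapon, explosion, ambient and other sidechain buses.
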MetaSoundSource /HotAirBalloon/Assets/Audio/MSS_HotAirBalloon_Hitmarker_Instigator
inputs [
 In Array {type=WaveAsset:Array}
 UE.Source.OnPlay {type=Trigger}
 UE.Spatialization.Azimuth {type=Float}
 UE.Spatialization.Elevation {type=Float}
]
interfaces [UE.OutputFormat.Stereo:1.0, UE.Source.OneShot:1.0, UE.Source:1.1, UE.Spatialization:1.0]
outputs [
 "UE.OutputFormat.Stereo.Audio:0" {type=Audio}
 "UE.OutputFormat.Stereo.Audio:1" {type=Audio}
 UE.Source.OneShot.OnFinished {type=Trigger}
]
pages [Default {
  input_defaults {
   UE.Source.OnPlay=null
   UE.Spatialization.Azimuth=0
   UE.Spatialization.Elevation=0
   In Array [
    /HotAirBalloon/Assets/Audio/Waves/Veh_HotAirBalloon_Hitmarker_Impt_01
    /HotAirBalloon/Assets/Audio/Waves/Veh_HotAirBalloon_Hitmarker_Impt_02
    /HotAirBalloon/Assets/Audio/Waves/Veh_HotAirBalloon_Hitmarker_Impt_03
    /HotAirBalloon/Assets/Audio/Waves/Veh_HotAirBalloon_Hitmarker_Impt_04
    /HotAirBalloon/Assets/Audio/Waves/Veh_HotAirBalloon_Hitmarker_Impt_05
   ]
  }
  nodes [
   ArrayRandomGet_0 = Array.Random Get:WaveAsset:Array(In Array=^In Array, Next=^UE.Source.OnPlay)
   WavePlayer_0 = UE.Wave Player:Stereo(Play=@ArrayRandomGet_0.On Next, Wave Asset=@ArrayRandomGet_0.Value) {also {version [1,1]}}
   AzimuthPannerStereo_0 = &AzimuthPannerStereo(Azimuth=^UE.Spatialization.Azimuth, In L=@WavePlayer_0.Out Left, In R=@WavePlayer_0.Out Right)
  ]
  links [
   {dst="@UE.OutputFormat.Stereo.Audio:0", src=AzimuthPannerStereo_0.Out L}
   {dst="@UE.OutputFormat.Stereo.Audio:1", src=AzimuthPannerStereo_0.Out R}
   {dst="@UE.Source.OneShot.OnFinished", src=WavePlayer_0.On Finished}
  ]
 }]
properties {
 assetImportData=&self:AssetImportData
 attenuationSettings=&/ATT_HotAirBalloon_Hitmarker
 numChannels=2
 outputFormat=Stereo
 sampleRate=48000
 soundClassObject=&Fort_Hitmarkers
 soundSubmixObject=&Fortnite_SFX
 volume=0.8
 busSends [
  {audioBus=&AB_Sidechain_FortWeapon_Vol, sendLevel=0.45}
  {audioBus=&AB_Sidechain_FortExplosion_Vol, sendLevel=0.4}
  {audioBus=&AB_Sidechain_FortAmbient_Vol, sendLevel=0.3}
  {audioBus=&AB_Sidechain_FortLoot_Vol, sendLevel=0.3}
  {audioBus=&AB_Sidechain_FortVehicles_Vol, sendLevel=0.3}
  {audioBus=&AB_Sidechain_FortCreatures_Vol, sendLevel=0.3}
  {audioBus=&AB_Sidechain_FortEmote_Vol, sendLevel=0.3}
 ]
 concurrencySet [/Game/Athena/Sounds/HitNotifications/Gunplay_Update/CON_Hitmarker_GunplayUpdate_Instigator_Standard]
 modulationSettings {
  volumeRouting=Union
  volumeModulationDestination {bEnableModulation=true, modulators [&Patch_FortHitNotify_Vol]}
 }
}
