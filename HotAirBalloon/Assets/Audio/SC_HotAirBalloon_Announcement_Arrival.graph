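# Sound cue for the hot air balloon's arrival announcement jingle.
# A single wave player plays the arrival jingle wave, with concurrency overrides capping simultaneous instances and setting a retrigger time.
# Base submix and submix sends are disabled, while a pre-effect send feeds the train station announcement audio bus borrowed from the Train content.
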SoundCue /HotAirBalloon/Assets/Audio/SC_HotAirBalloon_Announcement_Arrival
root=WavePlayer_0
nodes [{id=WavePlayer_0, properties {soundWaveAssetPtr=/HotAirBalloon/Assets/Audio/Waves/HotAirBalloon_Jingle_Arrival_S39_01}}]
properties {
 bEnableBaseSubmix=false
 bEnableSubmixSends=false
 maxDistance=2097152
 volumeMultiplier=1
 concurrencyOverrides {maxCount=2, retriggerTime=3}
 preEffectBusSends [{audioBus=/Train/Sound/AB_TrainStationAnnouncmentBus, sendLevel=0.6, soundSourceBus=null}]
}
