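# Sound cue for the hot air balloon's ascend sound, classed as an exterior vehicle sound.
# A Random node picks one of three equally weighted ascend wave variations.
# It uses the balloon burner attenuation and a concurrency override that allows only one instance at a time.
# It also sends at full level to the balloon's arrive/departure sidechain audio bus.
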SoundCue /HotAirBalloon/Assets/Audio/SC_HotAirBalloon_Ascend
root=Random_1
nodes [
 {id=Random_1, children [WavePlayer_11, WavePlayer_12, WavePlayer_13], properties {weights [1,1,1]}}
 {id=WavePlayer_11, properties {soundWaveAssetPtr=/HotAirBalloon/Assets/Audio/Waves/Veh_HotAirBalloon_Ascend_01}}
 {id=WavePlayer_12, properties {soundWaveAssetPtr=/HotAirBalloon/Assets/Audio/Waves/Veh_HotAirBalloon_Ascend_02}}
 {id=WavePlayer_13, properties {soundWaveAssetPtr=/HotAirBalloon/Assets/Audio/Waves/Veh_HotAirBalloon_Ascend_03}}
]
properties {
 attenuationSettings=&/ATT_HotAirBalloon_Burner
 bOverrideConcurrency=true
 maxDistance=9000
 soundClassObject=&Fort_Vehicles_Exterior
 volumeMultiplier=0.7
 busSends [{audioBus=&/AB_HotAirBalloon_ArriveDeparture_Sidechain, minSendLevel=0, sendLevel=1, soundSourceBus=null}]
 concurrencyOverrides {maxCount=1}
}
