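# Sound cue for the hot air balloon's bell, classed as exterior vehicle audio.
# It waits a fixed delay, then plays one of four equally weighted bell waves with slight random pitch and volume variation.
# It uses the lift-balloon attenuation, a three-per-owner prevent-new concurrency rule, and a send to an arrive/departure sidechain audio bus.
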SoundCue /HotAirBalloon/Assets/Audio/SC_HotAirBalloon_Bell
root=Delay_0
nodes [
 {id=Delay_0, children [Modulator_0], properties {delayMax=1.5, delayMin=1.5}}
 {id=Modulator_0, children [Random_0], properties {pitchMax=1.02, pitchMin=1, volumeMax=1.02, volumeMin=1}}
 {id=Random_0, children [WavePlayer_0,WavePlayer_1,WavePlayer_2,WavePlayer_3], properties {weights [1,1,1,1]}}
 {id=WavePlayer_0, properties {soundWaveAssetPtr=/HotAirBalloon/Assets/Audio/Waves/Veh_HotAirBalloon_Bell_01}}
 {id=WavePlayer_1, properties {soundWaveAssetPtr=/HotAirBalloon/Assets/Audio/Waves/Veh_HotAirBalloon_Bell_02}}
 {id=WavePlayer_2, properties {soundWaveAssetPtr=/HotAirBalloon/Assets/Audio/Waves/Veh_HotAirBalloon_Bell_03}}
 {id=WavePlayer_3, properties {soundWaveAssetPtr=/HotAirBalloon/Assets/Audio/Waves/Veh_HotAirBalloon_Bell_04}}
]
properties {
 attenuationSettings=&/ATT_HotAirBalloon_LiftBalloon
 maxDistance=3250
 soundClassObject=&Fort_Vehicles_Exterior
 volumeMultiplier=1
 busSends [{audioBus=&/AB_HotAirBalloon_ArriveDeparture_Sidechain, minSendLevel=0, sendLevel=1, soundSourceBus=null}]
 concurrencySet [/Game/Sounds/Concurrency_Rules/ThreeMaximum_Owner_PreventNew]
}
