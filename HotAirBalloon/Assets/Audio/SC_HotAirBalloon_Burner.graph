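# Sound cue for the hot air balloon burner, mixing a continuous burner loop with intermittent bursts under the exterior vehicle sound class.
# Bursts repeat through a Looping node, each picking one of ten short or long waves after a random delay, with slight pitch variation.
# The loop passes through two chained looping Envelopers of differing periods that drift pitch and volume, then its own Attenuation node.
# Cue-level attenuation and an owner-based concurrency rule also apply.
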SoundCue /HotAirBalloon/Assets/Audio/SC_HotAirBalloon_Burner
root=Mixer_0
nodes [
 {id=Attenuation_0, children [Enveloper_2], properties {attenuationSettings=&/ATT_HotAirBalloon_Burner}}
 {id=Delay_0, children [Random_0], properties {delayMax=20, delayMin=5}}
 {
  id=Enveloper_1
  children [WavePlayer_10]
  properties {
   bLoop=true
   bLoopIndefinitely=true
   loopEnd=30
   pitchCurve {editorCurveData {keys [{interpMode=RCIM_Cubic, time=0}, <at 15.9468 0.9181 RCIM_Cubic>, {interpMode=RCIM_Cubic, time=30}]}}
   volumeCurve {editorCurveData {keys [<at 0 1>]}}
  }
 }
 {
  id=Enveloper_2
  children [Enveloper_1]
  properties {
   bLoop=true
   bLoopIndefinitely=true
   loopEnd=20
   pitchCurve {editorCurveData {keys [{time=0}]}}
   volumeCurve {editorCurveData {keys [
      {arriveTangent=-0.01, interpMode=RCIM_Cubic, time=0, value=1}
      <at 10 0.7 RCIM_Cubic>
      {interpMode=RCIM_Cubic, leaveTangent=0.01, time=20, value=1}
     ]}}
  }
 }
 {id=Looping_0, children [Modulator_0]}
 {id=Mixer_0, children [Looping_0, Attenuation_0], properties {inputVolume [0.7,0.75]}}
 {id=Modulator_0, children [Delay_0], properties {pitchMax=1.02, pitchMin=0.98}}
 {
  id=Random_0
  children [WavePlayer_0,WavePlayer_1,WavePlayer_2,WavePlayer_3,WavePlayer_4,WavePlayer_5,WavePlayer_6,WavePlayer_7,WavePlayer_8,WavePlayer_9]
  properties {weights [1,1,1,1,1,1,1,1,1,1]}
 }
 {id=WavePlayer_0, properties {soundWaveAssetPtr=/HotAirBalloon/Assets/Audio/Waves/Veh_HotAirBalloon_Burner_Burst_Short_01}}
 {id=WavePlayer_1, properties {soundWaveAssetPtr=/HotAirBalloon/Assets/Audio/Waves/Veh_HotAirBalloon_Burner_Burst_Short_02}}
 {id=WavePlayer_10, properties {bLooping=true, soundWaveAssetPtr=/HotAirBalloon/Assets/Audio/Waves/Veh_HotairBalloon_Burner_Loop}}
 {id=WavePlayer_2, properties {soundWaveAssetPtr=/HotAirBalloon/Assets/Audio/Waves/Veh_HotAirBalloon_Burner_Burst_Short_03}}
 {id=WavePlayer_3, properties {soundWaveAssetPtr=/HotAirBalloon/Assets/Audio/Waves/Veh_HotAirBalloon_Burner_Burst_Short_04}}
 {id=WavePlayer_4, properties {soundWaveAssetPtr=/HotAirBalloon/Assets/Audio/Waves/Veh_HotAirBalloon_Burner_Burst_Short_05}}
 {id=WavePlayer_5, properties {soundWaveAssetPtr=/HotAirBalloon/Assets/Audio/Waves/Veh_HotAirBalloon_Burner_Burst_Long_01}}
 {id=WavePlayer_6, properties {soundWaveAssetPtr=/HotAirBalloon/Assets/Audio/Waves/Veh_HotAirBalloon_Burner_Burst_Long_02}}
 {id=WavePlayer_7, properties {soundWaveAssetPtr=/HotAirBalloon/Assets/Audio/Waves/Veh_HotAirBalloon_Burner_Burst_Long_03}}
 {id=WavePlayer_8, properties {soundWaveAssetPtr=/HotAirBalloon/Assets/Audio/Waves/Veh_HotAirBalloon_Burner_Burst_Long_04}}
 {id=WavePlayer_9, properties {soundWaveAssetPtr=/HotAirBalloon/Assets/Audio/Waves/Veh_HotAirBalloon_Burner_Burst_Long_05}}
]
properties {
 attenuationSettings=&ATT_HotAirBalloon_Burner
 maxDistance=9000
 soundClassObject=&Fort_Vehicles_Exterior
 volumeMultiplier=1
 concurrencySet [/Game/Sounds/Concurrency_Rules/ThreeMaximum_Owner_PreventNew]
}
