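# Sound cue for the hot air balloon's descend sound, classed as an exterior vehicle sound.
# It waits a fixed ten-second delay, then picks one of three equally weighted descend wave variations at random.
# It uses the balloon burner attenuation, allows only one instance at a time and sends to an arrive/departure sidechain audio bus.
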SoundCue /HotAirBalloon/Assets/Audio/SC_HotAirBalloon_Descend
root=Delay_0
nodes [
 {id=Delay_0, children [Random_1], properties {delayMax=10, delayMin=10}}
 {id=Random_1, children [WavePlayer_0,WavePlayer_1,WavePlayer_2], properties {weights [1,1,1]}}
 {id=WavePlayer_0, properties {soundWaveAssetPtr=/HotAirBalloon/Assets/Audio/Waves/Veh_HotAirBalloon_Descend_01}}
 {id=WavePlayer_1, properties {soundWaveAssetPtr=/HotAirBalloon/Assets/Audio/Waves/Veh_HotAirBalloon_Descend_02}}
 {id=WavePlayer_2, properties {soundWaveAssetPtr=/HotAirBalloon/Assets/Audio/Waves/Veh_HotAirBalloon_Descend_03}}
]
properties {
 attenuationSettings=&/ATT_HotAirBalloon_Burner
 bOverrideConcurrency=true
 maxDistance=9000
 soundClassObject=&Fort_Vehicles_Exterior
 volumeMultiplier=0.7
 busSends [{audioBus=&/AB_HotAirBalloon_ArriveDeparture_Sidechain, minSendLevel=0, sendLevel=1, soundSourceBus=null}]
 concurrencyOverrides {maxCount=1}
}
