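# Looping Sound Cue for the exterior floating ambience of a hot air balloon, attenuated as the lift balloon.
# A mixer blends a randomly picked floating loop with slow pitch drift, a quieter borrowed train rooftop wind with undulating volume, and intermittent balloon creaks.
# Creaks repeat through a looped random delay with slight pitch and volume modulation.
# A root enveloper fades the mix in, and a per-owner concurrency rule limits instances.
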SoundCue /HotAirBalloon/Assets/Audio/SC_HotAirBalloon_TopFloating_Loop
root=Enveloper_6
nodes [
 {id=Delay_0, children [Random_1], properties {delayMax=10, delayMin=5}}
 {
  id=Enveloper_0
  children [WavePlayer_2]
  properties {
   bLoop=true
   bLoopIndefinitely=true
   loopEnd=30
   pitchCurve {editorCurveData {keys [{time=0}]}}
   volumeCurve {editorCurveData {keys [
      {arriveTangent=-0.01, interpMode=RCIM_Cubic, time=0, value=1}
      <at 10 0.8 RCIM_Cubic>
      {interpMode=RCIM_Cubic, leaveTangent=0.01, time=20, value=1}
     ]}}
  }
 }
 {
  id=Enveloper_4
  children [Random_0]
  properties {
   bLoop=true
   bLoopIndefinitely=true
   loopEnd=30
   pitchCurve {editorCurveData {keys [{interpMode=RCIM_Cubic, time=0}, <at 15.9468 0.9181 RCIM_Cubic>, {interpMode=RCIM_Cubic, time=30}]}}
   volumeCurve {editorCurveData {keys [<at 0 1>]}}
  }
 }
 {
  id=Enveloper_6
  children [Mixer_0]
  properties {pitchCurve {editorCurveData {keys [{time=0}]}}, volumeCurve {editorCurveData {keys [<at 0 0 RCIM_Cubic>,<at 2 1>]}}}
 }
 {id=Looping_0, children [Modulator_0]}
 {id=Mixer_0, children [Enveloper_4,Enveloper_0,Looping_0], properties {inputVolume [1.5,0.3,1.5]}}
 {id=Modulator_0, children [Delay_0], properties {pitchMax=1.02, pitchMin=0.98, volumeMax=1.02, volumeMin=0.98}}
 {id=Random_0, children [WavePlayer_1,WavePlayer_0], properties {weights [1,1]}}
 {id=Random_1, children [WavePlayer_3,WavePlayer_4,WavePlayer_5,WavePlayer_6,WavePlayer_7], properties {weights [1,1,1,1,1]}}
 {id=WavePlayer_0, properties {bLooping=true, soundWaveAssetPtr=/HotAirBalloon/Assets/Audio/Waves/Veh_HotAirBalloon_Floating_Loop_02}}
 {id=WavePlayer_1, properties {bLooping=true, soundWaveAssetPtr=/HotAirBalloon/Assets/Audio/Waves/Veh_HotAirBalloon_Floating_Loop_01}}
 {id=WavePlayer_2, properties {soundWaveAssetPtr=/Train/Sound/MS/MSS_Train_Rooftop_Wind}}
 {id=WavePlayer_3, properties {soundWaveAssetPtr=/HotAirBalloon/Assets/Audio/Waves/Veh_HotAirBalloon_Balloon_Creak_01}}
 {id=WavePlayer_4, properties {soundWaveAssetPtr=/HotAirBalloon/Assets/Audio/Waves/Veh_HotAirBalloon_Balloon_Creak_02}}
 {id=WavePlayer_5, properties {soundWaveAssetPtr=/HotAirBalloon/Assets/Audio/Waves/Veh_HotAirBalloon_Balloon_Creak_03}}
 {id=WavePlayer_6, properties {soundWaveAssetPtr=/HotAirBalloon/Assets/Audio/Waves/Veh_HotAirBalloon_Balloon_Creak_04}}
 {id=WavePlayer_7, properties {soundWaveAssetPtr=/HotAirBalloon/Assets/Audio/Waves/Veh_HotAirBalloon_Balloon_Creak_05}}
]
properties {
 attenuationSettings=&/ATT_HotAirBalloon_LiftBalloon
 maxDistance=3250
 soundClassObject=&Fort_Vehicles_Exterior
 volumeMultiplier=1
 concurrencySet [/Game/Sounds/Concurrency_Rules/ThreeMaximum_Owner_PreventNew]
}
