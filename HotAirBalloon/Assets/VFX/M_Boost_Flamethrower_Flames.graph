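# Unlit translucent Niagara sprite material for the hot air balloon's boost flamethrower flames.
# A slowly rotating fire noise normal distorts the UVs of a panning water texture and a smoke mask.
# These combine into a sharpened flame shape that particle color tints and particle alpha scales.
# Dynamic parameters scale brightness and a distortion term, with the color fed through MF_FX_EmissiveAdjustments.
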Material /HotAirBalloon/Assets/VFX/M_Boost_Flamethrower_Flames
graph {
 properties {bUsedWithNiagaraSprites=true, blendMode=BLEND_Translucent, pixelDepthOffsetMode=PDOM_Legacy, shadingModel=MSM_Unlit, usedShadingModels=Unlit}
 nodes [
  n1 = TextureCoordinate() {uTiling=2.5, vTiling=2.5}
  n2 = Rotator(Coordinate=TextureCoordinate() {uTiling=0.25, vTiling=0.25}, Time=Multiply(A=Time(), B=Constant() {r=0.5})) {centerX=0.1, centerY=0.1, speed=-0.8}
  n3 = Multiply(
   A=ComponentMask(None=TextureSample(UVs=@n2) {texture=&/T_VCVFX_NoiseFire01_N}.RGB) {g=true, r=true}
   B=Constant() {r=0.05}
  )
  n4 = TextureSample(UVs=Panner(Coordinate=Add(A=@n1, B=@n3)) {speedX=0.1, speedY=0.5}) {texture=&/T_VCVFX_Water01_A}
  n5 = Add(A=@n4.RGB, B=Multiply(A=Constant() {r=3}, B=@n4.RGB))
  MaterialExpressionConstantClamp_2 = Saturate(None=Add(A=@n5, B=Constant() {r=-0.05}))
  n6 = TextureSample(UVs=Add(A=TextureCoordinate() {uTiling=4, vTiling=4}, B=@n3)) {texture=&/T_DLR_Flamethrower_SmokeMask}
  n7 = Multiply(A=Multiply(A=@MaterialExpressionConstantClamp_2, B=@n6.RGB), B=Constant() {r=1})
  n8 = Multiply(A=@n6.RGB, B=Constant() {r=0.35})
  MaterialExpressionConstantClamp_0 = Saturate(None=ConstantBiasScale(None=Add(A=@n7, B=@n8)) {bias=-0.1, scale=3})
  MaterialExpressionVertexColor_0 = ParticleColor()
  MaterialExpressionConstantClamp_1 = Saturate(None=Multiply(A=@MaterialExpressionConstantClamp_0, B=@MaterialExpressionVertexColor_0.A))
  n9 = DynamicParameter() {paramNames [DistortionAmount, Brightness, SoftAmount, NoiseAmount]}
  MaterialExpressionConstantClamp_3 = Saturate(None=Multiply(A=@n7, B=@n9.DistortionAmount))
  n10 = Multiply(A=@MaterialExpressionConstantClamp_0, B=@MaterialExpressionVertexColor_0.RGB)
  n11 = Add(A=@n4.RGB, B=@n5)
  MaterialFunctionCall(Emissive=Multiply(A=Multiply(A=@n11, B=@n10), B=@n9.Brightness)) {materialFunction=&MF_FX_EmissiveAdjustments}
  Multiply()
 ]
}
