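# Niagara system for hot air balloon idle wind: mesh streaks spawned around the owner, tinted by time of day.
# System-level smoothed velocity orients each mesh, and normalized velocity range sets spawn probability.
# One WindCurve emitter spawns particles in a 1000-unit sphere, with partition-dependent lifetime and two mesh renderers.
# Life curves and random values drive material parameters and alpha, with distance remaps fading streaks both very near and far away.
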NiagaraSystem /HotAirBalloon/Assets/VFX/NS_HotAirBalloon_Wind_Idle
dependencies {
 usedDataInterfaces [&NiagaraDataInterfaceCurve]
 usedDynamicInputs [
  &MakeVector
  &ReturnMeshOrientationAxis
  &OrientMeshToVector
  &TransformVector
  &ClampFloat
  &RemapRange
  &VectorLength
 ]
 usedModules [
  &SystemState
  &InterpolateOverTime
  &self:SystemSpawnScript.SystemScriptSource.SystemScriptGraph.NiagaraNodeAssignment_1.SetVariables_#1
  &self:SystemSpawnScript.SystemScriptSource.SystemScriptGraph.NiagaraNodeAssignment_0.SetVariables_#2
 ]
}
emitters {WindCurve {
  emitterUpdateScript [
   &EmitterState()
   &SpawnRate(Spawn Probability=$System.NormalizedVelocityRange, SpawnRate=15, Use Spawn Probability=true)
  ]
  particleSpawnScript [
   &PartitionParticles()
   &InitializeParticle(
    Color=&FNTimeOfDayColor(
     Night Color=<rgba 0.4667 0.5464 0.5833 0.4>
     Morning Color=<rgba 0.5958 0.6976 0.7448 0.6>
     Day Color=<rgba 0.8 0.9367 1 0.6>
     Evening Color=<rgba 0.575 0.6732 0.7188 0.6>
    )
    Lifetime=&RandomRangeFloat(
     Minimum=&MergeHelper_FloatSwitch(0=3, 1=1.5, 2=?, 3=?, Index=$Particles.PartitionParticles.Partition, Num=::NewEnumerator0)
     Maximum=2.5
    )
    Mesh Scale Mode=::NewEnumerator1
    Mesh Uniform Scale Max=8
    Mesh Uniform Scale Min=4
    UsePositionOffset=false
    Write Material Random=true
   )
   &ShapeLocation(
    MaskSpawnGroup=false
    Offset Mode=::NewEnumerator0
    Override Seed=false
    Sphere Radius=1000
    Sphere Surface Distribution=1
    UseOwnerScale=true
   )
   &InitialMeshOrientation(
    Mesh Orientation Mode=::NewEnumerator2
    Orientation Coordinate Space=::Local
    Orientation Vector=&NormalizeVector(Vector To Normalize=$System.SmoothedVelocity, Safe Normalize=true, Safe Fallback Threshold=0, Safe Fallback Vector=<0 0 0.001>)
    Use Rotation Vector=false
   )
  ]
  particleUpdateScript [
   &ParticleState()
   &DynamicMaterialParameters(
    Index 0 Param 1=&FloatFromCurve(FloatCurve=DataInterface() {propertyValues {Curve {keys [
         {arriveTangent=2.4535, interpMode=RCIM_Cubic, leaveTangent=2.4535, tangentMode=RCTM_User, time=0, value=0}
         {interpMode=RCIM_Cubic, tangentMode=RCTM_User, time=0.5, value=0.5}
         {arriveTangent=-2.5865, tangentMode=RCTM_User, time=1, value=0}
        ]}}})
    Index 0 Param 3=&Multiply_Float(
     A=&RandomRangeFloat(Minimum=3, Maximum=6)
     B=&Lerp_Float(
      A=0.75
      B=1
      Alpha=&FN_LOD_Distance_Remap(Start Distance=500, End Distance=5000, Minimum Output=0, Maximum Output=1, Clamp Output=true)
     )
    )
    Index 1 Param 1=&RandomRangeFloat(Minimum=0, Maximum=1)
    Index 1 Param 2=&RandomRangeFloat(Minimum=0, Maximum=1)
    Module.Index0Param0Write=true
    Module.Index0Param1Write=true
    Module.Index0Param2Write=true
    Module.Index0Param3Write=false
    Module.Index1Param0Write=true
    Module.Index1Param1Write=true
    Module.Index1Param2Write=true
    Module.Index1Param3Write=true
    Write Parameter Index 1=true
   )
   &ScaleColor(
    Scale Alpha=&FloatFromCurve(
     FloatCurve=DataInterface() {propertyValues {Curve {
        defaultValue=0
        keys [
         {interpMode=RCIM_Cubic, tangentMode=RCTM_User, time=0, value=0}
         {interpMode=RCIM_Cubic, tangentMode=RCTM_User, time=0.5, value=1}
         {tangentMode=RCTM_User, time=1, value=0}
        ]
       }}}
     Scale Curve=&Multiply_Float(
      A=&Lerp_Float(
       A=0
       B=1
       Alpha=&Multiply_Float(
        A=&FN_LOD_Distance_Remap(Start Distance=0, End Distance=75, Minimum Output=0, Maximum Output=1, Clamp Output=true)
        B=&FN_LOD_Distance_Remap(Start Distance=2500, End Distance=10000, Minimum Output=1, Maximum Output=0, Clamp Output=true)
       )
      )
      B=&RandomRangeFloat(Minimum=0.1, Maximum=0.25)
     )
    )
    ScaleA=true
    ScaleRGB=true
   )
   &SolveForcesAndVelocity(Clamp Velocity=false, Limit Acceleration=false)
  ]
  rendererClasses [&NiagaraMeshRendererProperties]
  renderers [
   {rendererClass=&NiagaraMeshRendererProperties, rendererIndex=0}
   {rendererClass=&NiagaraMeshRendererProperties, rendererIndex=1}
  ]
 }}
summary {
 systemName=NS_HotAirBalloon_Wind_Idle
 emitters [{emitterName=WindCurve, rendererClasses [&NiagaraMeshRendererProperties]}]
 userVariables []
}
